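# Material function for the headset back bling that recolors an input material and adds flame glow as a stage transition progresses.
# Transition Amount blends First, Middle and Final Stage colors, which tint the desaturated, brightened base color.
# The same color drives additive emissive through a Fresnel that broadens with progress, masked by Alpha.
# A panning world-aligned flame texture, sampled at Local Pos offset by Gmap, breaks up that glow.
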MaterialFunction /Game/Accessories/FORT_Backpacks/Backpack_M_MED_Headset/Materials/MF_Headset_StageTransition
graph {nodes [
  n1 = FunctionInput() {inputName=Transition Amount, inputType=FunctionInput_Scalar}
  Divide(A=@n1)
  n2 = GetMaterialAttributes(None=FunctionInput() {bUsePreviewValueAsDefault=true, inputName=Material, inputType=FunctionInput_MaterialAttributes}) {attributeGetTypes [#1, #2]}
  n3 = Saturate(None=@n1)
  n4 = MaterialFunctionCall(
   A=VectorParameter() {parameterName=First Stage, defaultValue=<rgba 1 0 0.02 1>}.RGB
   B=VectorParameter() {parameterName=Middle Stage, defaultValue=<rgba 0.2368 0 1 1>}.RGB
   C=VectorParameter() {parameterName=Final Stage, defaultValue=<rgba 0.1139 0 0.6094 1>}.RGB
   Alpha=@n3
  ) {materialFunction=&Lerp_3Color}
  n5 = Multiply(A=@n4.Result, B=Fresnel(ExponentIn=LinearInterpolate(Alpha=@n3) {constA=1.66}) {baseReflectFraction=0, exponent=1})
  n6 = FunctionInput() {inputName=Gmap, inputType=FunctionInput_Scalar}
  n7 = ScalarParameter() {defaultValue=0.3, parameterName=offset}
  n8 = MaterialFunctionCall!(
   TextureObject=TextureObject() {samplerType=SAMPLERTYPE_LinearGrayscale, texture=/Game/Effects/Fort_Effects/Emotes/UltraLight/Texture/T_FluidFlameTile_Greyscale}
   TextureSize=ScalarParameter() {defaultValue=20, parameterName=Size}
   WorldPosition=Add(A=Multiply(A=@n6, B=@n7), B=FunctionInput() {inputName=Local Pos})
   Export Float 4
   In=ScalarParameter() {defaultValue=0.45, parameterName=Panning}
   World Space Normal
   ProjectionTransitionContrast
  ) {materialFunction=/Game/Athena/SafeZone/WorldAlignedTexture_ScalePan}
  n9 = Multiply(
   A=Multiply(A=Multiply(A=@n5, B=@n8.XYZ Texture), B=ScalarParameter() {defaultValue=6, parameterName=FlameBrightness})
   B=FunctionInput() {bUsePreviewValueAsDefault=true, inputName=Alpha, inputType=FunctionInput_Scalar}
  )
  n10 = Multiply(A=Desaturation(None=@n2.BaseColor, Fraction=Constant() {r=1}), B=@n4.Result)
  n11 = Constant() {r=3}
  FunctionOutput(None=SetMaterialAttributes!(MaterialAttributes=@n2.MaterialAttributes, EmissiveColor=Add(A=@n2.EmissiveColor, B=@n9), BaseColor=Multiply(A=@n10, B=@n11)) {attributeSetTypes [#1, #2]})
  Multiply(A=@n2.EmissiveColor, B=@n4.Result)
 ]}
