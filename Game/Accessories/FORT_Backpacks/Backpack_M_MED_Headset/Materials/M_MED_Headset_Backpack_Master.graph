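# Master material for the MED Headset backpack, a Default Lit skeletal-mesh material built on material attributes and the shared character PreFX and PostFX functions.
# A saturated Backpack_Transition scalar drives MF_Headset_StageTransition, using pre-skinned local position and the SkinFX texture's red channel.
# The green channel, offset by that scalar and rounded up, masks a wrap fresnel emissive applied to the transition result before PostFX.
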Material /Game/Accessories/FORT_Backpacks/Backpack_M_MED_Headset/Materials/M_MED_Headset_Backpack_Master
graph {
 properties {
  bUseMaterialAttributes=true
  bUsedWithMeshDeformer=true
  bUsedWithSkeletalMesh=true
  pixelDepthOffsetMode=PDOM_Legacy
  usedShadingModels=Default Lit
 }
 nodes [
  ComponentMask() {g=true, r=true}
  n1 = MaterialFunctionCall() {materialFunction=&BaseFN_CharacterFunction_PreFX}
  n2 = Saturate(None=ScalarParameter() {parameterName=Backpack_Transition})
  n3 = TextureSampleParameter2D() {parameterName=SkinFX, texture=/Game/Accessories/FORT_Backpacks/Backpack_M_MED_Headset/Textures/Backpack_Base_color}
  n4 = MaterialFunctionCall!(
   Transition Amount=@n2
   Local Pos=@n1.Pre skinned Local Position
   Gmap=@n1.Pre skinned Local Position
   Alpha=@n3.R
   Material=@n1.Pre skinned Local Position
  ) {materialFunction=&/MF_Headset_StageTransition}
  n5 = Add(A=Subtract(A=@n3.G) {constB=0.99}, B=@n2)
  MaterialFunctionCall(
   AO=@n1.AO
   Distance Blend=@n1.AO
   Fresnel=@n1.AO
   Gmap=@n1.AO
   Pre Skinned Local Position=@n1.AO
   Material=MaterialFunctionCall!(
    Materials=@n4.Result
    Normal=GetMaterialAttributes(None=@n4.Result) {attributeGetTypes [#1]}.Normal
    Primary Mask=Ceil(None=@n5)
   ) {materialFunction=/Game/Weapons/WeaponSkins/Wraps/MaterialFunctions/MF_FresnelEmissive}.Result
  ) {materialFunction=&BaseFN_CharacterFunction_PostFX}
  MaterialFunctionCall!(In) {materialFunction=/Game/Weapons/WeaponSkins/MaterialFunction/MF_EmissiveFresnel}
 ]
}
